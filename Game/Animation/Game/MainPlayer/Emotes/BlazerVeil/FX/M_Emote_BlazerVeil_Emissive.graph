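# Unlit translucent material for glowing Niagara mesh particles in the BlazerVeil emote, colored by particle color through an emissive adjustment function.
# Opacity combines particle alpha with switchable UV, view-angle and depth fades, optionally blending in panned noise along a UV gradient.
# A default-on Avoid Clipping switch masks bands near the V edges.
# Optional world position offset pushes vertices along normals plus a local forward vector, weighted by inverted UV and driven by dynamic parameters.
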Material /Game/Animation/Game/MainPlayer/Emotes/BlazerVeil/FX/M_Emote_BlazerVeil_Emissive
graph {
 properties {
  bUsedWithNiagaraMeshParticles=true
  blendMode=BLEND_Translucent
  pixelDepthOffsetMode=PDOM_Legacy
  previewMesh=/BRCosmetics/Accessories/FORT_Backpacks/Backpack_M_MED_SkippingClouds/FX/SM_Backpack_SkippingClouds_LeftEyeGlow
  shadingModel=MSM_Unlit
  usedShadingModels=Unlit
 }
 nodes [
  n1 = ParticleColor()
  MaterialFunctionCall(Emissive=@n1.RGB) {materialFunction=&MF_FX_EmissiveAdjustments}
  n2 = StaticComponentMaskParameter(None=TextureCoordinate()) {defaultR=true, parameterName=UV Fade Channel}
  n3 = Constant() {r=1}
  n4 = StaticSwitchParameter(True=Power(Base=@n2, Exp=ScalarParameter() {defaultValue=1, parameterName=UV Fade Sensitivity, sliderMax=5}), False=@n3) {parameterName=UseUVFade}
  n5 = TextureCoordinate()
  n6 = Panner(
   Coordinate=Multiply(
    A=StaticComponentMaskParameter(None=@n5) {defaultG=true, parameterName=Texture Channel}
    B=ScalarParameter() {defaultValue=5, parameterName=UV Scale, sliderMax=20}
   ) {constB=10}
   Speed=AppendVector(
    A=ScalarParameter() {defaultValue=1, parameterName=UV Speed X, sliderMax=5}
    B=ScalarParameter() {defaultValue=1, parameterName=UV Speed Y, sliderMax=5}
   )
  ) {speedX=0.5, speedY=0.5}
  n7 = Multiply(
   A=@n4
   B=Power(
    Base=TextureSampleParameter2D(UVs=@n6) {parameterName=Texture, samplerType=SAMPLERTYPE_Grayscale, texture=/FortWater/Textures/LowResBlurredNoise}.R
    Exp=ScalarParameter() {defaultValue=1, parameterName=Texture Contrast, sliderMax=10}
   )
  )
  n8 = ScalarParameter() {defaultValue=1, parameterName=View Angle Fade Sensitivity, sliderMax=5}
  n9 = Multiply(
   A=StaticSwitchParameter(
    True=MaterialFunctionCall("FadeSensitivity (optional)"=@n8) {materialFunction=&FadeBasedOnViewAngle}.Result
    False=@n3
   ) {parameterName=UseViewAngleFade}
   B=@n4
  )
  n10 = Power(Base=@n2, Exp=ScalarParameter() {defaultValue=1, parameterName=UV Lerp Power, sliderMax=10})
  n11 = ComponentMask(None=TextureCoordinate()) {g=true}
  n12 = Constant() {r=0.04}
  n13 = Constant() {r=0.13}
  n14 = Constant() {r=0.5}
  n15 = TextureCoordinate()
  n16 = Max(
   A=SphereMask(A=Add(A=@n11) {constB=-1}, B=@n12, Radius=@n13, Hardness=@n14) {attenuationRadius=0.1, hardnessPercent=0}
   B=SphereMask(A=ComponentMask(None=@n15) {g=true}, B=@n12, Radius=@n13, Hardness=@n14) {attenuationRadius=0.1, hardnessPercent=0}
  )
  n17 = Multiply(
   A=StaticSwitchParameter(True=LinearInterpolate(A=@n7, B=@n9, Alpha=@n10), False=@n9) {parameterName=UseTexture}
   B=StaticSwitchParameter(True=OneMinus(None=@n16), False=Constant() {r=1}) {defaultValue=true, parameterName=Avoid Clipping}
  )
  n18 = ScalarParameter() {parameterName=FadeDistance}
  Multiply(A=@n1.A, B=StaticSwitchParameter(True=DepthFade(Opacity=@n17, FadeDistance=@n18), False=@n17) {parameterName=UseDepthFade})
  n19 = VectorParameter() {parameterName=WPO Forward Local Vector, defaultValue=<rgba 0 1 0 1>}
  n20 = DynamicParameter() {defaultValue=<rgba 0 1 1 1>, paramNames [WPO, WPO Forward, Param3, Param4]}
  n21 = Add(A=VertexNormalWS(), B=Multiply(A=Transform(None=@n19.RGB) {transformSourceType=TRANSFORMSOURCE_Local}, B=@n20.WPO Forward))
  StaticSwitchParameter(True=Multiply(A=OneMinus(None=@n2), B=Multiply(A=@n21, B=@n20.WPO)), False=Constant()) {parameterName=UseWPO}
 ]
}
